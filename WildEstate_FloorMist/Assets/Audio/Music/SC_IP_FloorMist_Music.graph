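# Sound Cue that plays the VelvetThunder music stinger for the FloorMist item or feature in the WildEstate_FloorMist plugin.
# A single Wave Player node feeds the output directly, at half volume.
# A dedicated FloorMist music attenuation asset spatializes it with a long max distance, and it routes through the Fortnite POI music sound class.
# It also carries an audio asset user data object.
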SoundCue /WildEstate_FloorMist/Assets/Audio/Music/SC_IP_FloorMist_Music
root=WavePlayer_0
nodes [{id=WavePlayer_0, properties {soundWaveAssetPtr=&/MX_S38_IP_VelvetThunder_STNG}}]
properties {
 attenuationSettings=&/ATT_FloorMist_Music
 maxDistance=28000
 soundClassObject=/Game/Sounds/FortSoundClassesAndModes/Fort_Music_POI_PSM
 volumeMultiplier=0.5
 assetUserData [&self:AudioAssetUserData_0]
}
